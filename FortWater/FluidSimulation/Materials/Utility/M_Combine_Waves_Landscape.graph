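# Unlit utility material in the fluid simulation set that combines a wave height field with landscape data, apparently for render-target passes.
# It unpacks RG8 height from SeedRT and takes the maximum of that and the WaveRT red channel raised by a Wave Z offset.
# That height is packed with LandscapeRT's green and blue channels into a three-channel value, and negative emissive is allowed.
# A few leftover nodes feed nothing.
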Material /FortWater/FluidSimulation/Materials/Utility/M_Combine_Waves_Landscape
graph {
 properties {bAllowNegativeEmissiveColor=true, pixelDepthOffsetMode=PDOM_Legacy, shadingModel=MSM_Unlit, usedShadingModels=Unlit}
 nodes [
  n1 = TextureSampleParameter2D() {
   parameterName=SeedRT
   samplerType=SAMPLERTYPE_LinearColor
   texture=/Engine/Functions/Engine_MaterialFunctions02/PivotPainter2/Black_1x1_EXR_Texture
  }
  n2 = MaterialFunctionCall!(RG=@n1.RGB) {materialFunction=/Landmass/Landscape/BlueprintBrushes/MF/UnpackRG8Height}
  n3 = Add(
   A=TextureSampleParameter2D() {parameterName=WaveRT, samplerType=SAMPLERTYPE_LinearColor, texture=/FortWater/FluidSimulation/Textures/RT_A}.R
   B=ScalarParameter() {defaultValue=2500, parameterName=Wave Z}
  )
  n4 = TextureSampleParameter2D() {
   parameterName=LandscapeRT
   samplerType=SAMPLERTYPE_LinearColor
   texture=&Black_1x1_EXR_Texture
  }
  AppendVector(A=AppendVector(A=Max(A=@n2.Result, B=@n3), B=@n4.G), B=@n4.B)
  Constant()
  Multiply(A=@n2.Result) {constB=2}
  ScalarParameter() {defaultValue=1000, parameterName=sasdf}
 ]
}
